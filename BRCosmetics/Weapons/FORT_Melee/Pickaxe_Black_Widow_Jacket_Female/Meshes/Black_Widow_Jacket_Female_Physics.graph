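# Physics asset for the female Black Widow Jacket pickaxe, defining collision bodies and joints for the handle and its dangling cloth bone chain.
# The two-handed handle body combines a long capsule with a box, while the cloth uses a small root capsule followed by four boxes.
# Five constraints link the chain in series from the handle, sharing angular limits that are tighter on one swing axis than on the other swing and twist.
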PhysicsAsset /BRCosmetics/Weapons/FORT_Melee/Pickaxe_Black_Widow_Jacket_Female/Meshes/Black_Widow_Jacket_Female_Physics
bodies {
 Melee_TwoHanded_Handle [
  {shapeType=Capsule, length=114.2117, radius=2.1438, center=<0.5656 -0.6165 30.6475>, rotation=<rot -0.131 -89.9981 0.0267>}
  {shapeType=Box, extentX=7.5, extentY=7.5, extentZ=7.5}
 ]
 dyn_cloth [{shapeType=Capsule, length=2.1193, radius=1.2296, center=<0.0004 -0.8036 0.3654>, rotation=<rot -10.0261 -89.9715 0.0001>}]
 dyn_cloth_01 [{shapeType=Box, extentX=1.193, extentY=1.8614, extentZ=6.1642}]
 dyn_cloth_02 [{shapeType=Box, extentX=1.193, extentY=2.3111, extentZ=7.9015, center=<0 0 0.353>}]
 dyn_cloth_03 [{shapeType=Box, extentX=1.193, extentY=2.662, extentZ=7.7984}]
 dyn_cloth_04 [{shapeType=Box, extentX=1.193, extentY=2.9583, extentZ=9.0177, center=<0 0 -1.1163>}]
}
constraints [
 {bone1Name=dyn_cloth_01, bone2Name=dyn_cloth, swing1LimitDegrees=10, swing2LimitDegrees=45, twistLimitDegrees=45}
 {bone1Name=dyn_cloth_02, bone2Name=dyn_cloth_01, swing1LimitDegrees=10, swing2LimitDegrees=45, twistLimitDegrees=45}
 {bone1Name=dyn_cloth_03, bone2Name=dyn_cloth_02, swing1LimitDegrees=10, swing2LimitDegrees=45, twistLimitDegrees=45}
 {bone1Name=dyn_cloth_04, bone2Name=dyn_cloth_03, swing1LimitDegrees=10, swing2LimitDegrees=45, twistLimitDegrees=45}
 {bone1Name=dyn_cloth, bone2Name=Melee_TwoHanded_Handle, swing1LimitDegrees=10, swing2LimitDegrees=45, twistLimitDegrees=45}
]
